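# Nanite-enabled job board body material, built between the 2023 character-parent PreFX and PostFX functions, with masked color tinting from defaults and custom primitive data.
# A custom-primitive-data timestamp drives a square-root-eased lid rotation about a hard-coded hinge, masked by red vertex color.
# Blue vertex color masks an antenna offset toward a fixed local height.
# Both deformations use World Position Offset, with destructible damage wobble added last.
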Material /JobBoard/Materials/JobBoard/M_JobBoard_Body
graph {
 properties {bUseMaterialAttributes=true, bUsedWithNanite=true, pixelDepthOffsetMode=PDOM_Legacy, usedShadingModels=Default Lit}
 nodes [
  n1 = MaterialFunctionCall() {materialFunction=&MF_CharacterParent_2023_PreFX}
  n2 = GetMaterialAttributes(None=@n1.Result) {attributeGetTypes [#1]}
  n3 = VectorParameter() {
   desc=Use this color to set a default color for when there's no Primitive Data set on the mesh.
   group=JobBoard
   parameterName=ColorTint
   defaultValue=<rgba 0.2 0.2 0.2 1>
  }
  n4 = VectorParameter() {bUseCustomPrimitiveData=true, group=JobBoard, parameterName=ColorTint_CPD, primitiveDataIndex=1, defaultValue=<rgba 1 1 1 1>}
  n5 = Multiply(
   A=Multiply(A=@n2.Base Color, B=ScalarParameter() {defaultValue=1, group=JobBoard, parameterName=BaseColorBoost})
   B=LinearInterpolate(A=@n3.RGB, B=@n4.RGB, Alpha=@n4.RGB)
  )
  n6 = GetMaterialAttributes(None=BlendMaterialAttributes(
    A=@n2.MaterialAttributes
    B=SetMaterialAttributes!(MaterialAttributes=@n2.MaterialAttributes, Base Color=@n5) {attributeSetTypes [#1]}
    Alpha=ChannelMaskParameter(None=@n1.Tech Art Mask) {group=JobBoard, parameterName=ColorTintMaskChannel}.R
   )) {attributeGetTypes [#2]}
  n7 = ScalarParameter() {bUseCustomPrimitiveData=true, group=JobBoard, parameterName=LidAnimation_CPD}
  n8 = Subtract(A=Time(), B=@n7)
  n9 = Saturate(None=Divide(A=Abs(None=@n8), B=ScalarParameter() {defaultValue=0.5, group=JobBoard, parameterName=LidAnimationSpeed}) {constB=0.5})
  n10 = Saturate(None=@n7)
  n11 = SquareRoot(None=NamedRerouteDeclaration(None=Multiply(A=@n9, B=@n10)) {name=Time, nodeColor=<rgba 1 0.4922 0 1>})
  n12 = MaterialFunctionCall!(
   RotationAngle=Multiply(A=NamedRerouteDeclaration(None=@n11) {name=Squared Time, nodeColor=<rgba 0 0.6016 1 1>}) {constB=0.2745}
   NormalizedRotationAxis=Transform(None=Constant3Vector() {desc=Hinge Axis Direction, constant=<rgba -0.1182 0.993 0 1>}) {transformSourceType=TRANSFORMSOURCE_Local}
   PivotPoint=TransformPosition(None=Constant3Vector() {desc=Hinge Pivot, constant=<rgba -44.0818 27.7993 106.942 1>}) {transformType=TRANSFORMPOSSOURCE_World}
   Position=WorldPosition().XYZ
   Normal=Transform(None=@n6.Normal)
   TangentU
   TangentV
  ) {materialFunction=/Game/Characters/Player/Male/Medium/Bodies/M_MED_Inferno/FX/Mat/RotateVertexAboutAxis}
  n13 = Transform(None=@n12.Normal) {transformSourceType=TRANSFORMSOURCE_World, transformType=TRANSFORM_Tangent}
  n14 = GetMaterialAttributes(None=BlendMaterialAttributes(
    A=@n6.MaterialAttributes
    B=SetMaterialAttributes!(MaterialAttributes=@n6.MaterialAttributes, World Position Offset=@n12.WPO, Normal=@n13) {attributeSetTypes [#3, #2]}
    Alpha=VertexColor().R
   )) {attributeGetTypes [#3]}
  n15 = MaterialFunctionCall() {materialFunction=&LocalPosition}
  n16 = TransformPosition(None=AppendVector(
    A=ComponentMask(None=@n15.Local Position) {g=true, r=true}
    B=Constant3Vector() {desc=Antenna Pivot, constant=<rgba 20.3756 50.7667 103.649 0>}.B
   )) {transformType=TRANSFORMPOSSOURCE_World}
  n17 = WorldPosition()
  n18 = Add(A=@n14.World Position Offset, B=Multiply(A=Subtract(A=@n16, B=@n17.XYZ), B=NamedRerouteUsage()))
  n19 = GetMaterialAttributes(None=BlendMaterialAttributes(
    A=@n14.MaterialAttributes
    B=SetMaterialAttributes!(MaterialAttributes=@n14.MaterialAttributes, World Position Offset=@n18) {attributeSetTypes [#3]}
    Alpha=VertexColor().B
   )) {attributeGetTypes [#3]}
  n20 = MaterialFunctionCall() {materialFunction=&MF_FN_DestructibleDamage}
  MaterialFunctionCall(
   Material=SetMaterialAttributes(MaterialAttributes=@n19.MaterialAttributes, World Position Offset=Add(A=@n19.World Position Offset, B=@n20.DamageAnimWPO)) {attributeSetTypes [#3]}
   M=@n1.M
   Pre Skinned Local Position=@n1.M
   Distance Blend=@n1.M
   Fresnel=@n1.M
  ) {materialFunction=&MF_CharacterParent_2023_PostFX}
 ]
}
