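# Nanite-enabled, lit material for a job board screen, showing a tinted emissive icon with view-angle dimming.
# An icon layer joins a switchable radial-gradient or noise LCD backlight, with intermittent panning-noise UV distortion and scrolling scanlines masked by player distance.
# World position offset combines destructible damage with a hinged lid rotation timed from custom primitive data.
# Curie fallback and Creative move-tool functions wrap the output, with simplified specular and roughness on low-quality mobile.
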Material /JobBoard/Materials/JobBoard/M_JobBoard_Screen
graph {
 properties {bUseMaterialAttributes=true, bUsedWithNanite=true, pixelDepthOffsetMode=PDOM_Legacy, usedShadingModels=Default Lit}
 nodes [
  n1 = VectorParameter() {group=ScreenFX, parameterName=ColorTint, sortPriority=1, defaultValue=<rgba 0.3185 0.552 0.9301 1>}
  n2 = Constant() {r=-1}
  n3 = MaterialFunctionCall(Float3=MaterialFunctionCall() {materialFunction=&ObjectPivotPoint}.Object Pivot Location) {materialFunction=&BreakOutFloat3Components}
  n4 = Frac(None=Multiply(A=@n3.R, B=@n3.R))
  n5 = Multiply(
   A=TextureCoordinate() {uTiling=0.5, vTiling=0.5}
   B=ScalarParameter() {defaultValue=0.75, group=ScreenFX, parameterName=BacklightTiling, sortPriority=3}
  )
  n6 = Add(
   A=TextureSample(UVs=Add(A=@n4, B=@n5)) {samplerType=SAMPLERTYPE_Grayscale, texture=&LowResBlurredNoise}.RGB
   B=Multiply(A=NamedRerouteUsage()) {constB=2}
  )
  n7 = StaticSwitchParameter(True=NamedRerouteUsage(), False=Constant()) {defaultValue=true, group=ScreenFX, parameterName=UseRadialGradientBacklight, sortPriority=3}
  n8 = ScalarParameter() {defaultValue=0.2, group=ScreenFX, parameterName=BacklightStrength, sortPriority=3}
  n9 = LinearInterpolate(A=Multiply(A=@n8) {constB=0.4}, B=@n8, Alpha=NamedRerouteUsage())
  n10 = MaterialFunctionCall(
   UVs=LinearInterpolate(
    A=TextureCoordinate()
    B=NamedRerouteUsage()
    Alpha=ScalarParameter() {defaultValue=0.15, group=ScreenFX, parameterName=DistortIconScale, sortPriority=4}
   ) {constAlpha=0.2}
   Texture Scale=AppendVector(
    A=ScalarParameter() {defaultValue=0.45, group=ScreenFX, parameterName=IconScaleU, sortPriority=2}
    B=ScalarParameter() {defaultValue=0.8, group=ScreenFX, parameterName=IconScaleV, sortPriority=2}
   )
  ) {materialFunction=&ScaleUVsByCenter}
  n11 = TextureSampleParameter2D(UVs=@n10.UVs) {
   group=ScreenFX
   parameterName=Icon
   samplerSource=SSM_Clamp_WorldGroupSettings
   sortPriority=0
   texture=/Game/Athena/UI/Challenges/Art/ChallengeTypeIcons/UI_ChallengeIcon_Visit
  }
  n12 = MaterialFunctionCall!(UVs=Saturate(None=@n10.UVs), Size=Constant() {r=0.49}, SizeScale) {materialFunction=/Game/Packages/Fortress_SharedMaterials/Base_Material_Functions/SquarePatternFromUVs}
  n13 = Add(
   A=Multiply(
    A=StaticSwitchParameter(True=@n6, False=@n7) {group=ScreenFX, parameterName=UseLCDBacklight, sortPriority=3}
    B=StaticSwitchParameter(True=@n9, False=@n8) {group=ScreenFX, parameterName=UseFadeBacklight}
   )
   B=Multiply(
    A=Multiply(A=@n11.A, B=@n12.PatternOutput)
    B=ScalarParameter() {defaultValue=1, group=ScreenFX, parameterName=IconEmissiveScale, sortPriority=2}
   )
  )
  n14 = Multiply(A=Multiply(A=Desaturation(None=@n1.RGB, Fraction=@n2)) {constB=0.6}, B=@n13)
  n15 = LinearInterpolate(
   A=ScalarParameter() {group=ScreenFX, parameterName=FresnelMin, sortPriority=2}
   B=ScalarParameter() {defaultValue=1, group=ScreenFX, parameterName=FresnelMax, sortPriority=2}
   Alpha=DotProduct(A=CameraVectorWS(), B=VertexNormalWS())
  )
  n16 = Multiply(A=Multiply(A=@n13, B=@n1.RGB), B=ScalarParameter() {defaultValue=5, group=ScreenFX, parameterName=EmissiveScale, sortPriority=1})
  n17 = Panner(Coordinate=TextureCoordinate()) {speedY=0.02}
  n18 = Frac(None=Multiply(A=ComponentMask(None=@n17) {g=true}, B=ScalarParameter() {defaultValue=20, group=ScreenFX, parameterName=ScanlineRepeat, sortPriority=8}))
  n19 = Constant() {r=0.5}
  n20 = ScalarParameter() {defaultValue=0.4, group=ScreenFX, parameterName=ScanlineRadius, sortPriority=9}
  n21 = ScalarParameter() {defaultValue=0.2, group=ScreenFX, parameterName=ScanlineHarndess, sortPriority=10}
  n22 = CollectionParameter() {collection=&FortniteMaterialParameters, parameterName=PlayerPosition}
  n23 = ComponentMask(None=@n22) {b=true, g=true, r=true}
  n24 = WorldPosition()
  n25 = ScalarParameter() {defaultValue=800, group=ScreenFX, parameterName=FadeRadius, sortPriority=5}
  n26 = ScalarParameter() {defaultValue=0.75, group=ScreenFX, parameterName=FadeHardness, sortPriority=5}
  n27 = LinearInterpolate(
   A=Constant() {r=1}
   B=LinearInterpolate(
    B=SphereMask(A=@n18, B=@n19, Radius=@n20, Hardness=@n21)
    Alpha=ScalarParameter() {defaultValue=1, group=ScreenFX, parameterName=ScanlineStrength, sortPriority=7}
   ) {constA=1}
   Alpha=NamedRerouteDeclaration(None=SphereMask(A=@n23, B=@n24.XYZ, Radius=@n25, Hardness=@n26) {attenuationRadius=350, hardnessPercent=80}) {name=DistanceFadeAlpha, nodeColor=<rgba 1 0 0.7734 1>}
  )
  n28 = Multiply(A=StaticSwitchParameter(True=Multiply(A=@n16, B=@n27), False=@n16) {defaultValue=true, group=ScreenFX, parameterName=UseScanlines}, B=@n15)
  n29 = ScalarParameter() {bUseCustomPrimitiveData=true, parameterName=LidAnimation_CPD}
  n30 = Subtract(A=Time(), B=@n29)
  n31 = Saturate(None=Divide(A=Abs(None=@n30), B=ScalarParameter() {defaultValue=0.5, parameterName=LidAnimationSpeed}) {constB=0.5})
  n32 = Saturate(None=@n29)
  n33 = SquareRoot(None=NamedRerouteDeclaration(None=Multiply(A=@n31, B=@n32)) {name=Time, nodeColor=<rgba 1 0.4922 0 1>})
  n34 = MaterialFunctionCall!(
   RotationAngle=Multiply(A=NamedRerouteDeclaration(None=@n33) {name=Squared Time, nodeColor=<rgba 0 0.6016 1 1>}) {constB=0.2745}
   NormalizedRotationAxis=Transform(None=Constant3Vector() {desc=Hinge Axis Direction, constant=<rgba -0.1182 0.993 0 1>}) {transformSourceType=TRANSFORMSOURCE_Local}
   PivotPoint=TransformPosition(None=Constant3Vector() {desc=Hinge Pivot, constant=<rgba -44.0818 27.7993 106.942 1>}) {transformType=TRANSFORMPOSSOURCE_World}
   Position=WorldPosition().XYZ
   Normal=Transform(None=Constant3Vector() {constant=<rgba 0 0 1 0>})
   TangentU
   TangentV
  ) {materialFunction=/Game/Characters/Player/Male/Medium/Bodies/M_MED_Inferno/FX/Mat/RotateVertexAboutAxis}
  n35 = MaterialFunctionCall() {materialFunction=&MF_FN_DestructibleDamage}
  n36 = MaterialFunctionCall!(
   In=MakeMaterialAttributes!(
    BaseColor=Multiply(A=@n14, B=@n15)
    Metallic=ScalarParameter() {defaultValue=0.5, group=ScreenFX, parameterName=Metallic, sortPriority=1}
    Specular=ScalarParameter() {defaultValue=0.2, group=ScreenFX, parameterName=Specular, sortPriority=1}
    Roughness=ScalarParameter() {defaultValue=0.15, group=ScreenFX, parameterName=Roughness, sortPriority=1}
    Anisotropy
    EmissiveColor=EyeAdaptationInverse(LightValueInput=@n28).EyeAdaptationInverse
    Opacity
    OpacityMask
    Normal=Transform(None=@n34.Normal) {transformSourceType=TRANSFORMSOURCE_World, transformType=TRANSFORM_Tangent}
    Tangent
    WorldPositionOffset=Add(A=@n35.DamageAnimWPO, B=@n34.WPO)
    SubsurfaceColor
    ClearCoat
    ClearCoatRoughness
    AmbientOcclusion
    Refraction
    CustomizedUVs_0
    CustomizedUVs_1
    CustomizedUVs_2
    CustomizedUVs_3
    CustomizedUVs_4
    CustomizedUVs_5
    CustomizedUVs_6
    CustomizedUVs_7
    PixelDepthOffset
    ShadingModel
    Displacement
   )
   UseSphericalMappingOverride
  ) {materialFunction=/Game/Packages/Fortress_SharedMaterials/Base_Material_Functions/MF_CurieFallbackEffect}
  Composite!(InputPin=@n36.Result) {subgraphName=Mobile Opt}
  MaterialExpressionReroute_0 = Reroute(None=@n36.Result)
  n37 = GetMaterialAttributes(None=@MaterialExpressionReroute_0.InputPin)
  n38 = Constant() {r=0.5}
  n39 = Constant() {r=0.8}
  n40 = ShadingPathSwitch(
   Default=@MaterialExpressionReroute_0.InputPin
   Mobile=QualitySwitch(
    Default=@MaterialExpressionReroute_0.InputPin
    Low=SetMaterialAttributes!(MaterialAttributes=@n37.MaterialAttributes, Specular=@n38, Roughness=@n39) {attributeSetTypes [#1, #2]}
    High=@MaterialExpressionReroute_0.InputPin
    Medium=@MaterialExpressionReroute_0.InputPin
    Epic=@MaterialExpressionReroute_0.InputPin
   )
  )
  MaterialExpressionReroute_1 = Reroute(None=@n40)
  MaterialFunctionCall(MA In=@MaterialExpressionReroute_1.Output) {materialFunction=&MF_CreativeMoveTool}
  n41 = TextureCoordinate()
  n42 = TextureCoordinate()
  n43 = AppendVector(
   A=ScalarParameter() {group=ScreenFX, parameterName=DistortTilingX, sortPriority=4}
   B=ScalarParameter() {defaultValue=1, group=ScreenFX, parameterName=DistortTilingY, sortPriority=4}
  )
  n44 = TextureSampleParameter2D(UVs=Panner(Coordinate=Multiply(A=@n42, B=@n43)) {bFractionalPart=true, speedX=0.015, speedY=0.384}) {
   group=ScreenFX
   parameterName=DistortTexture
   samplerType=SAMPLERTYPE_Normal
   sortPriority=4
   texture=/Game/Global/Textures/Default/Noise/64x64Noise
  }
  n45 = ScalarParameter() {defaultValue=0.1, group=ScreenFX, parameterName=DistortScale, sortPriority=4}
  n46 = MaterialFunctionCall(Float3=MaterialFunctionCall() {materialFunction=&ObjectPivotPoint}.Object Pivot Location) {materialFunction=&BreakOutFloat3Components}
  n47 = Add(A=Multiply(A=@n46.R, B=@n46.R), B=Time())
  n48 = Add(A=Sine(None=@n47) {period=0.355}, B=Sine(None=@n47) {period=0.8755})
  n49 = ComponentMask(None=Power(Base=Saturate(None=@n48)) {constExponent=5}) {r=true}
  n50 = Multiply(A=AppendVector(A=@n44.R, B=Constant()), B=VertexInterpolator(VS=Multiply(A=@n45, B=@n49)).PS)
  NamedRerouteDeclaration(None=StaticSwitchParameter(True=Add(A=@n41, B=@n50), False=@n41) {defaultValue=true, group=ScreenFX, parameterName=UseDistortion, sortPriority=4}) {name=DistortedUVs, nodeColor=<rgba 0 0.7187 1 1>}
  n51 = NamedRerouteUsage()
  n52 = AppendVector(
   A=ScalarParameter() {defaultValue=0.5, group=ScreenFX, parameterName=BacklightU, sortPriority=3}
   B=ScalarParameter() {defaultValue=0.5, group=ScreenFX, parameterName=BacklightV, sortPriority=3}
  )
  NamedRerouteDeclaration(None=MaterialFunctionCall(
    UVs=MaterialFunctionCall(UVs=@n51, Texture Scale=@n52) {materialFunction=&ScaleUVsByCenter}.UVs
    Radius=ScalarParameter() {defaultValue=1.2952, group=ScreenFX, parameterName=BacklightRadius, sliderMax=2, sortPriority=3}
    Density=ScalarParameter() {defaultValue=0.85, group=ScreenFX, parameterName=BacklightDensity, sliderMax=10, sortPriority=3}
   ) {materialFunction=&RadialGradientExponential}.RadialGradientExponential) {name=RadialGradient, nodeColor=<rgba 0 0.2031 1 1>}
  PinBase() {reroutePins [InputPin {expression=&self:MaterialExpressionReroute_0}]}
  PinBase!(Output=@n40) {reroutePins [Output {expression=&self:MaterialExpressionReroute_1}]}
 ]
}
